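# Control Rig for the male medium Teriyaki Warrior body that drives corrective deform bones as the character poses.
# A four-way sequence runs the shared DeformRig function, right and left foot correctives weighted by spherical pose readers, then neck and collar fixes.
# Collar bones blend parent constraints with offsets weighted by remapped head-to-spine_05 Euler angles, after shrinking an anti-clipping neck bone.
# A separate foot-rotation branch is disconnected from execution.
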ControlRigBlueprint /Game/Characters/Player/Male/Medium/Bodies/M_MED_Teriyaki_Warrior/Meshes/M_MED_TeriyakiWarrior_Body_CtrlRig
graphs [
 RigVMFunctionLibrary {links [], nodes []}
 RigVMFunction_Sequence_1_ContainedGraph {nodes [
   n1 = RigVMFunctionEntryNode()
   RigVMFunction_Sequence = &RigVMFunction_Sequence(ExecuteContext=@n1.ExecuteContext)
   n2 = &RigVMFunction_Sequence(ExecuteContext=@RigVMFunction_Sequence.B)
   n3 = &RigVMFunction_Sequence(ExecuteContext=@n2.B)
   RigVMFunctionReturnNode(A=@RigVMFunction_Sequence.A, B=@n2.A, C=@n3.A, D=@n3.B)
  ]}
 RigVMModel {nodes [
   RigUnit_BeginExecution = &RigUnit_BeginExecution()
   RigVMFunction_Sequence = RigVMAggregateNode(ExecuteContext=@RigUnit_BeginExecution.ExecutePin)
   FN_DeformRig_FNC = RigVMFunctionReferenceNode(
    DeformRig_ Curve="(deformRig_l_upperarm_up_alpha_72_#1=False,deformRig_r_upperarm_up_alpha_73_#2=False,deformRig_l_upperarm_fr_alpha_74_#3=False,deformRig_r_upperarm_fr_alpha_75_CB51B8 ...(truncated)"
    DeformRig_Bool="(upperarm_1_#4=True,elbow_3_#5=True,wrist_7_#6=True,groin_8_#7=True,glute_17_#8=True,knee_9_40C408324CA261 ...(truncated)"
    DeformRig_Debug=(upperarm_debug_18_#4=False,wrist_debug_20_#6=False,lat_debug_26_#9=False,groinGlute_debug_27_#7=False)
    DeformRig_Value="(upperarm_up_18_#10=(Rotation=(X=0,Y=0,Z=0,W=1),Translation=(X=0,Y=0,Z=8),Scale3D=(X=0.7,Y=1,Z=1)),upperarm_fwd_fr_19_#11=(Rotation=(X=0,Y=0,Z=0,W=1),Translation=(X=0,Y=10,Z=0),Scale3D=(X=0.5,Y=1,Z= ...(truncated)"
    ExecuteContext=@RigVMFunction_Sequence.A
   )
   n1 = &RigUnit_SphericalPoseReader(
    ActiveRegionScaleFactors=(PositiveWidth=0.5,NegativeWidth=0.5,PositiveHeight=2,NegativeHeight=2)
    ActiveRegionSize=0.2
    Debug=(bDrawDebug=false,bDraw2D=False,bDrawLocalAxes=False,DebugScale=25,DebugSegments=20,DebugThickness=0.25)
    DriverAxis=(X=1,Y=0,Z=0)
    DriverItem=(Type=Bone,Name="foot_r")
    FalloffRegionScaleFactors=(PositiveWidth=3.55,NegativeWidth=1,PositiveHeight=1,NegativeHeight=1)
    FalloffSize=0.2
    FlipHeightScaling=False
    FlipWidthScaling=False
    OptionalParentItem=(Type=Bone,Name="None")
    RotationOffset=(X=90,Y=0,Z=-8)
    ExecutePin=@RigVMFunction_Sequence.B
   )
   > &RigUnit_ModifyTransforms(
    ItemToModify="((Item=(Type=Bone,Name=\"deform_foot_in_r\"),Transform=(Rotation=(X=0,Y=-0.0687,Z=0,W=-0.9976),Translation=(X=-3.0596,Y=0,Z=-1.481),Scale3D=(X=2.6346,Y=1,Z=2.2))),(Item=(Type=Bone,Name=\"deform_foot_out_r\"),Transform=(Rotation=(X=0,Y=0.4439,Z=0,W=0.8961),Tran ...(truncated)"
    WeightMinimum=0
    Weight=@n1.OutputParam
   )
   > n2 = &RigUnit_SphericalPoseReader(
    ActiveRegionScaleFactors=(PositiveWidth=0.5,NegativeWidth=0.5,PositiveHeight=2,NegativeHeight=2)
    ActiveRegionSize=0.2
    Debug=(bDrawDebug=false,bDraw2D=False,bDrawLocalAxes=False,DebugScale=25,DebugSegments=20,DebugThickness=0.25)
    DriverAxis=(X=1,Y=0,Z=0)
    DriverItem=(Type=Bone,Name="foot_r")
    FalloffRegionScaleFactors=(PositiveWidth=1.35,NegativeWidth=1,PositiveHeight=1,NegativeHeight=1)
    FalloffSize=0.2
    FlipHeightScaling=False
    FlipWidthScaling=False
    OptionalParentItem=(Type=Bone,Name="None")
    RotationOffset=(X=90,Y=0,Z=17)
   )
   > &RigUnit_ModifyTransforms(
    ItemToModify="((Item=(Type=Bone,Name=\"deform_foot_in_r\"),Transform=(Rotation=(X=0,Y=0,Z=0,W=1),Translation=(X=0,Y=0,Z=0),Scale3D=(X=1,Y=1,Z=1))),(Item=(Type=Bone,Name=\"deform_foot_out_r\"),Transform=(Rotation=(X=0,Y=0.3111,Z=0,W=0.9504),Translation=(X=58.1728,Y=0,Z=-0.27 ...(truncated)"
    WeightMinimum=0
    Weight=@n2.OutputParam
   )
   > n3 = &RigUnit_SphericalPoseReader(
    ActiveRegionScaleFactors=(PositiveWidth=1,NegativeWidth=1,PositiveHeight=1,NegativeHeight=1)
    ActiveRegionSize=0.1
    Debug=(bDrawDebug=False,bDraw2D=False,bDrawLocalAxes=False,DebugScale=25,DebugSegments=20,DebugThickness=0.25)
    DriverAxis=(X=1,Y=0,Z=0)
    DriverItem=(Type=Bone,Name="foot_r")
    FalloffRegionScaleFactors=(PositiveWidth=3.5,NegativeWidth=1,PositiveHeight=1,NegativeHeight=1)
    FalloffSize=0.2
    FlipHeightScaling=False
    FlipWidthScaling=False
    OptionalParentItem=(Type=Bone,Name="None")
    RotationOffset=(X=-90,Y=0,Z=8)
   )
   > &RigUnit_ModifyTransforms(
    ItemToModify="((Item=(Type=Bone,Name=\"deform_foot_in_r\"),Transform=(Rotation=(X=0,Y=0.548,Z=0,W=0.8365),Translation=(X=4.3129,Y=0,Z=1.3027),Scale3D=(X=2.2,Y=1,Z=0.1))),(Item=(Type=Bone,Name=\"deform_foot_out_r\"),Transform=(Rotation=(X=0,Y=-0.357,Z=0,W=0.9341),Translation ...(truncated)"
    WeightMinimum=0
    Weight=@n3.OutputParam
   )
   n4 = &RigUnit_SphericalPoseReader(
    ActiveRegionScaleFactors=(PositiveWidth=0.5,NegativeWidth=0.5,PositiveHeight=2,NegativeHeight=2)
    ActiveRegionSize=0.2
    Debug=(bDrawDebug=false,bDraw2D=False,bDrawLocalAxes=False,DebugScale=25,DebugSegments=20,DebugThickness=0.25)
    DriverAxis=(X=1,Y=0,Z=0)
    DriverItem=(Type=Bone,Name="foot_l")
    FalloffRegionScaleFactors=(PositiveWidth=3.55,NegativeWidth=1,PositiveHeight=1,NegativeHeight=1)
    FalloffSize=0.2
    FlipHeightScaling=False
    FlipWidthScaling=False
    OptionalParentItem=(Type=Bone,Name="None")
    RotationOffset=(X=90,Y=0,Z=-8)
    ExecutePin=@RigVMFunction_Sequence.C
   )
   > &RigUnit_ModifyTransforms(
    ItemToModify="((Item=(Type=Bone,Name=\"deform_foot_in_l\"),Transform=(Rotation=(X=0,Y=-0.0687,Z=0,W=-0.9976),Translation=(X=-3.0596,Y=0,Z=-1.481),Scale3D=(X=2.6346,Y=1,Z=2.2))),(Item=(Type=Bone,Name=\"deform_foot_out_l\"),Transform=(Rotation=(X=0,Y=0.4439,Z=0,W=0.8961),Tran ...(truncated)"
    WeightMinimum=0
    Weight=@n4.OutputParam
   )
   > n5 = &RigUnit_SphericalPoseReader(
    ActiveRegionScaleFactors=(PositiveWidth=0.5,NegativeWidth=0.5,PositiveHeight=2,NegativeHeight=2)
    ActiveRegionSize=0.2
    Debug=(bDrawDebug=false,bDraw2D=False,bDrawLocalAxes=False,DebugScale=25,DebugSegments=20,DebugThickness=0.25)
    DriverAxis=(X=1,Y=0,Z=0)
    DriverItem=(Type=Bone,Name="foot_l")
    FalloffRegionScaleFactors=(PositiveWidth=1.35,NegativeWidth=1,PositiveHeight=1,NegativeHeight=1)
    FalloffSize=0.2
    FlipHeightScaling=False
    FlipWidthScaling=False
    OptionalParentItem=(Type=Bone,Name="None")
    RotationOffset=(X=90,Y=0,Z=17)
   )
   > &RigUnit_ModifyTransforms(
    ItemToModify="((Item=(Type=Bone,Name=\"deform_foot_in_l\"),Transform=(Rotation=(X=0,Y=0,Z=0,W=1),Translation=(X=0,Y=0,Z=0),Scale3D=(X=1,Y=1,Z=1))),(Item=(Type=Bone,Name=\"deform_foot_out_l\"),Transform=(Rotation=(X=0,Y=0.3111,Z=0,W=0.9504),Translation=(X=58.1728,Y=0,Z=-0.27 ...(truncated)"
    WeightMinimum=0
    Weight=@n5.OutputParam
   )
   > n6 = &RigUnit_SphericalPoseReader(
    ActiveRegionScaleFactors=(PositiveWidth=1,NegativeWidth=1,PositiveHeight=1,NegativeHeight=1)
    ActiveRegionSize=0.1
    Debug=(bDrawDebug=false,bDraw2D=False,bDrawLocalAxes=False,DebugScale=25,DebugSegments=20,DebugThickness=0.25)
    DriverAxis=(X=1,Y=0,Z=0)
    DriverItem=(Type=Bone,Name="foot_l")
    FalloffRegionScaleFactors=(PositiveWidth=3.5,NegativeWidth=1,PositiveHeight=1,NegativeHeight=1)
    FalloffSize=0.2
    FlipHeightScaling=False
    FlipWidthScaling=False
    OptionalParentItem=(Type=Bone,Name="None")
    RotationOffset=(X=-90,Y=0,Z=8)
   )
   > &RigUnit_ModifyTransforms(
    ItemToModify="((Item=(Type=Bone,Name=\"deform_foot_in_l\"),Transform=(Rotation=(X=0,Y=0.548,Z=0,W=0.8365),Translation=(X=4.3129,Y=0,Z=1.3027),Scale3D=(X=2.2,Y=1,Z=0.1))),(Item=(Type=Bone,Name=\"deform_foot_out_l\"),Transform=(Rotation=(X=0,Y=-0.357,Z=0,W=0.9341),Translation ...(truncated)"
    WeightMinimum=0
    Weight=@n6.OutputParam
   )
   &RigUnit_ModifyTransforms(
    ItemToModify=((Item=(Type=Bone,Name="neckContact_reduceClipping"),Transform=(Rotation=(X=0,Y=0,Z=0,W=1),Translation=(X=0,Y=0,Z=0),Scale3D=(X=0.5,Y=0.5,Z=0.5))))
    Weight=1
    WeightMinimum=0
    ExecutePin=@RigVMFunction_Sequence.D
   )
   > &RigUnit_ParentConstraint(
    AdvancedSettings=(InterpolationType=Average,RotationOrderForFilter=XZY)
    Child=(Type=Bone,Name="deform_collarBack_base")
    Parents="((Item=(Type=Bone,Name=\"spine_03\"),Weight=0),(Item=(Type=Bone,Name=\"spine_04\"),Weight=0),(Item=(Type=Bone,Name=\"clavicle_r\"),Weight=0),(Item=(Type=Bone,Name=\"spine_05\"),Weight=0.1974),(Item=(Type=Bone,Name=\"clavicle_l\"),Weight=0),(Item=(Type=Bone,Name=\"nec ...(truncated)"
    Weight=1
   )
   RigVMFunction_MathQuaternionToEuler = &RigVMFunction_MathQuaternionToEuler(
    RotationOrder=ZYX
    Value=&RigUnit_GetRelativeTransformForItem(Child=(Type=Bone,Name="head"), Parent=(Type=Bone,Name="spine_05"), bChildInitial=False, bParentInitial=False).RelativeTransform.Rotation
   )
   Remap = &RigVMFunction_MathDoubleRemap(SourceMaximum=-35, SourceMinimum=0, TargetMaximum=1, TargetMinimum=0, Value=@RigVMFunction_MathQuaternionToEuler.Result.Z)
   > &RigUnit_ModifyTransforms(
    ItemToModify="((Item=(Type=Bone,Name=\"deform_collarBack\"),Transform=(Rotation=(X=0.5684,Y=0,Z=0,W=-0.8227),Translation=(X=-0.0002,Y=-19.0619,Z=13.402),Scale3D=(X=1,Y=1,Z=1))),(Item=(Type=Bone,Name=\"deform_collarSideL\"),Transform=(Rotation=(X=0,Y=0,Z=0,W=-1),Translation= ...(truncated)"
    WeightMinimum=0
    Weight=@Remap.Result
   )
   > &RigUnit_ParentConstraint(
    AdvancedSettings=(InterpolationType=Average,RotationOrderForFilter=XZY)
    Child=(Type=Bone,Name="deform_collarFront")
    Parents="((Item=(Type=Bone,Name=\"spine_03\"),Weight=0.0298),(Item=(Type=Bone,Name=\"spine_04\"),Weight=0.0242),(Item=(Type=Bone,Name=\"clavicle_r\"),Weight=0.0585),(Item=(Type=Bone,Name=\"spine_05\"),Weight=0.1974),(Item=(Type=Bone,Name=\"clavicle_l\"),Weight=0.1128),(Item= ...(truncated)"
    Weight=0.5
   )
   > &RigUnit_ParentConstraint(
    AdvancedSettings=(InterpolationType=Average,RotationOrderForFilter=XZY)
    Child=(Type=Bone,Name="deform_collarSideR_base")
    Parents="((Item=(Type=Bone,Name=\"spine_03\"),Weight=0.0298),(Item=(Type=Bone,Name=\"spine_04\"),Weight=0.0242),(Item=(Type=Bone,Name=\"clavicle_r\"),Weight=0.0585),(Item=(Type=Bone,Name=\"spine_05\"),Weight=0.1974),(Item=(Type=Bone,Name=\"clavicle_l\"),Weight=0.1128),(Item= ...(truncated)"
    Weight=0.5
   )
   > &RigUnit_ParentConstraint(
    AdvancedSettings=(InterpolationType=Average,RotationOrderForFilter=XZY)
    Child=(Type=Bone,Name="deform_collarSideL_base")
    Parents="((Item=(Type=Bone,Name=\"spine_03\"),Weight=0.0298),(Item=(Type=Bone,Name=\"spine_04\"),Weight=0.0242),(Item=(Type=Bone,Name=\"clavicle_r\"),Weight=0.0585),(Item=(Type=Bone,Name=\"spine_05\"),Weight=0.1974),(Item=(Type=Bone,Name=\"clavicle_l\"),Weight=0.1128),(Item= ...(truncated)"
    Weight=0.5
   )
   > &RigUnit_ModifyTransforms(
    ItemToModify=((Item=(Type=Bone,Name="deform_collarSideR"),Transform=(Rotation=(X=0,Y=0,Z=0,W=1),Translation=(X=-8.4736,Y=0.0479,Z=-6.0726),Scale3D=(X=1,Y=1,Z=1))))
    WeightMinimum=0
    Weight=&RigVMFunction_MathDoubleRemap(SourceMaximum=60, SourceMinimum=0, TargetMaximum=1, TargetMinimum=0, Value=@RigVMFunction_MathQuaternionToEuler.Result.X).Result
   )
   > &RigUnit_ModifyTransforms(
    ItemToModify="((Item=(Type=Bone,Name=\"deform_collarSideL\"),Transform=(Rotation=(X=0,Y=0,Z=0,W=1),Translation=(X=4.1273,Y=-4.6134,Z=-0.8942),Scale3D=(X=1,Y=1,Z=1))),(Item=(Type=Bone,Name=\"deform_collarSideR\"),Transform=(Rotation=(X=0,Y=-0.669,Z=0,W=0.7432),Translation=(X ...(truncated)"
    WeightMinimum=0
    Weight=&RigVMFunction_MathDoubleRemap(SourceMaximum=-60, SourceMinimum=0, TargetMaximum=1, TargetMinimum=0, Value=@RigVMFunction_MathQuaternionToEuler.Result.X).Result
   )
   n7 = &RigVMFunction_Sequence(ExecuteContext=())
   &RigUnit_ModifyTransforms(
    ItemToModify="((Item=(Type=Bone,Name=\"foot_l\"),Transform=(Rotation=(X=0,Y=0,Z=0.5314,W=0.8471),Translation=(X=0,Y=0,Z=0),Scale3D=(X=1,Y=1,Z=1))),(Item=(Type=Bone,Name=\"foot_r\"),Transform=(Rotation=(X=0,Y=0,Z=0.5146,W=0.8574),Translation=(X=0,Y=0,Z=0),Scale3D=(X=1,Y=1,Z= ...(truncated)"
    Weight=1
    WeightMinimum=0
    ExecutePin=@n7.A
   )
  ]}
]
hierarchy [
 root {loc=<0 0 0>, rot=<0 0 0 1>}
 attach {parent=root, loc=<0 0 0>, rot=<0 0 0 1>}
 pelvis {parent=root, loc=<0 -0.7071 93.1191>, rot=<0.0127 0.707 -0.0127 -0.707>}
 spine_01 {parent=pelvis, loc=<6.7494 -0.6388 0>, rot=<0 0 0.034 -0.9994>}
 spine_02 {parent=spine_01, loc=<7.2518 -0.035 0>, rot=<0 0 0.0238 -0.9997>}
 spine_03 {parent=spine_02, loc=<6.0104 0 0>, rot=<0 0 0.0149 -0.9999>}
 spine_04 {parent=spine_03, loc=<8.1024 0 0>, rot=<0 0 -0.0132 -0.9999>}
 spine_05 {parent=spine_04, loc=<10.0146 0 0>, rot=<0 0 -0.0015 -1>}
 clavicle_l {parent=spine_05, loc=<4.3575 -0.1466 -2.3223>, rot=<0.0465 -0.7289 -0.009 -0.6829>}
 upperarm_l {parent=clavicle_l, loc=<15.5111 0 0>, rot=<0.0226 -0.3234 -0.0088 -0.9459>}
 lowerarm_l {parent=upperarm_l, loc=<26.6514 0 0>, rot=<0.0391 -0.0099 -0.2453 -0.9686>}
 hand_l {parent=lowerarm_l, loc=<22.3491 0 0>, rot=<-0.7081 -0.0507 0.0975 -0.6975>}
 index_metacarpal_l {parent=hand_l, loc=<1.6535 0.0183 -2.96>, rot=<0.0158 -0.0124 -0.0173 -0.9996>}
 index_01_l {parent=index_metacarpal_l, loc=<8.3008 0.0357 -0.0524>, rot=<-0.0044 0.0485 0.0924 -0.9945>}
 index_02_l {parent=index_01_l, loc=<3.3587 -0.3004 0.0452>, rot=<-0.002 0.0027 0.3225 -0.9466>}
 index_03_l {parent=index_02_l, loc=<2.5578 -0.123 -0.0252>, rot=<-0.002 0.0067 0.2558 -0.9667>}
 middle_metacarpal_l {parent=hand_l, loc=<2.1856 -0.1725 -0.5984>, rot=<0.006 0.0168 -0.0329 -0.9993>}
 middle_01_l {parent=middle_metacarpal_l, loc=<7.3275 -0.2447 0.0129>, rot=<-0.0208 0.0042 0.0858 -0.9961>}
 middle_02_l {parent=middle_01_l, loc=<4.0123 0.1401 0.1175>, rot=<-0.0218 0.0258 0.3512 -0.9357>}
 middle_03_l {parent=middle_02_l, loc=<2.6077 0.0534 0.1373>, rot=<0.0082 0.0188 0.1708 -0.9851>}
 pinky_metacarpal_l {parent=hand_l, loc=<1.7616 -0.9753 3.3814>, rot=<-0.0397 0.0814 -0.069 -0.9935>}
 pinky_01_l {parent=pinky_metacarpal_l, loc=<7.1709 -0.4831 -0.1094>, rot=<-0.0654 -0.0035 0.2242 -0.9723>}
 pinky_02_l {parent=pinky_01_l, loc=<2.8274 0.0893 -0.1043>, rot=<0.0037 0.0011 0.2854 -0.9584>}
 pinky_03_l {parent=pinky_02_l, loc=<2.2046 0.0307 -0.0078>, rot=<-0.0002 0.0098 0.1445 -0.9895>}
 ring_metacarpal_l {parent=hand_l, loc=<2.2074 -0.3625 1.3638>, rot=<-0.0284 0.0722 -0.0458 -0.9959>}
 ring_01_l {parent=ring_metacarpal_l, loc=<7.0295 -0.6197 -0.1446>, rot=<-0.0489 -0.0252 0.1892 -0.9804>}
 ring_02_l {parent=ring_01_l, loc=<3.6174 0.4647 -0.1027>, rot=<0 0 0.2806 -0.9598>}
 ring_03_l {parent=ring_02_l, loc=<2.4152 0.114 0.0199>, rot=<0 0 0.1189 -0.9929>}
 thumb_01_l {parent=hand_l, loc=<1.3118 -0.3191 -2.459>, rot=<0.6546 -0.4042 0.0664 -0.6354>}
 thumb_02_l {parent=thumb_01_l, loc=<5.9336 -0.0448 0.0401>, rot=<-0.0037 0.0213 0.298 -0.9543>}
 thumb_03_l {parent=thumb_02_l, loc=<2.2223 0.3682 -0.0907>, rot=<0.0118 -0.013 0.0066 -0.9998>}
 weapon_l {parent=hand_l, loc=<-0.1974 -3.4111 0.3811>, rot=<0 0 0 1>}
 lowerarm_twist_01_l {parent=lowerarm_l, loc=<17.8549 0 0>, rot=<0 0 0 1>}
 lowerarm_twist_02_l {parent=lowerarm_l, loc=<9.6199 0 0>, rot=<0 0 0 1>}
 upperarm_twist_01_l {parent=upperarm_l, loc=<0.0041 0 0>, rot=<0 0 0 1>}
 deform_upperarm_bk_base_l {parent=upperarm_twist_01_l, loc=<1.0813 -7.2193 -1.9453>, rot=<0 0 0 1>}
 deform_upperarm_bk_l {parent=deform_upperarm_bk_base_l, loc=<0 0 0>, rot=<0 0 0 1>}
 deform_upperarm_fr_base_l {parent=upperarm_twist_01_l, loc=<-0.3691 7.0506 -2.765>, rot=<0 0 0 1>}
 deform_upperarm_fr_l {parent=deform_upperarm_fr_base_l, loc=<0 0 0>, rot=<0 0 0 1>}
 deform_upperarm_up_base_l {parent=upperarm_twist_01_l, loc=<-3.2216 -1.1391 3.3311>, rot=<0 0 0 1>}
 deform_upperarm_up_l {parent=deform_upperarm_up_base_l, loc=<0 0 0>, rot=<0 0 0 1>}
 upperarm_twist_02_l {parent=upperarm_l, loc=<13.7941 0 0>, rot=<0 0 0 1>}
 clavicle_r {parent=spine_05, loc=<4.3575 -0.1466 2.3223>, rot=<0.7289 0.0465 -0.6829 0.009>}
 upperarm_r {parent=clavicle_r, loc=<-15.5111 0 0>, rot=<0.0226 -0.3234 -0.0088 -0.9459>}
 lowerarm_r {parent=upperarm_r, loc=<-26.6515 0 0>, rot=<0.0391 -0.0099 -0.2453 -0.9686>}
 hand_r {parent=lowerarm_r, loc=<-22.3492 0 0>, rot=<-0.7081 -0.0507 0.0975 -0.6975>}
 index_metacarpal_r {parent=hand_r, loc=<-1.6534 -0.0181 2.9601>, rot=<0.0158 -0.0124 -0.0173 -0.9996>}
 index_01_r {parent=index_metacarpal_r, loc=<-8.3008 -0.0357 0.0524>, rot=<-0.0044 0.0485 0.0924 -0.9945>}
 index_02_r {parent=index_01_r, loc=<-3.3589 0.3003 -0.0453>, rot=<-0.002 0.0027 0.3225 -0.9466>}
 index_03_r {parent=index_02_r, loc=<-2.5578 0.1231 0.0251>, rot=<-0.002 0.0067 0.2558 -0.9667>}
 middle_metacarpal_r {parent=hand_r, loc=<-2.1856 0.1726 0.5984>, rot=<0.006 0.0168 -0.0329 -0.9993>}
 middle_01_r {parent=middle_metacarpal_r, loc=<-7.3273 0.2447 -0.0129>, rot=<-0.0208 0.0042 0.0858 -0.9961>}
 middle_02_r {parent=middle_01_r, loc=<-4.0142 -0.1411 -0.07>, rot=<-0.0217 0.0259 0.3512 -0.9357>}
 middle_03_r {parent=middle_02_r, loc=<-2.6074 -0.0534 -0.1312>, rot=<0.0083 0.0188 0.1708 -0.9851>}
 pinky_metacarpal_r {parent=hand_r, loc=<-1.7617 0.9753 -3.3814>, rot=<-0.0397 0.0814 -0.069 -0.9935>}
 pinky_01_r {parent=pinky_metacarpal_r, loc=<-7.1707 0.4832 0.1094>, rot=<-0.0654 -0.0035 0.2242 -0.9723>}
 pinky_02_r {parent=pinky_01_r, loc=<-2.8274 -0.0892 0.1043>, rot=<0.0037 0.0011 0.2854 -0.9584>}
 pinky_03_r {parent=pinky_02_r, loc=<-2.2045 -0.031 0.0078>, rot=<-0.0002 0.0098 0.1445 -0.9895>}
 ring_metacarpal_r {parent=hand_r, loc=<-2.2073 0.3627 -1.3638>, rot=<-0.0284 0.0722 -0.0458 -0.9959>}
 ring_01_r {parent=ring_metacarpal_r, loc=<-7.0297 0.6197 0.1446>, rot=<-0.0489 -0.0252 0.1892 -0.9804>}
 ring_02_r {parent=ring_01_r, loc=<-3.6171 -0.4648 0.1027>, rot=<0 0 0.2806 -0.9598>}
 ring_03_r {parent=ring_02_r, loc=<-2.4153 -0.1141 -0.0199>, rot=<0 0 0.1189 -0.9929>}
 thumb_01_r {parent=hand_r, loc=<-1.3119 0.3192 2.459>, rot=<0.6546 -0.4042 0.0664 -0.6354>}
 thumb_02_r {parent=thumb_01_r, loc=<-5.9335 0.0447 -0.0401>, rot=<-0.0037 0.0213 0.298 -0.9543>}
 thumb_03_r {parent=thumb_02_r, loc=<-2.2223 -0.3682 0.0907>, rot=<0.0118 -0.013 0.0066 -0.9998>}
 weapon_r {parent=hand_r, loc=<0.1975 3.4112 -0.3811>, rot=<0 0 0 -1>}
 lowerarm_twist_01_r {parent=lowerarm_r, loc=<-17.855 0 0>, rot=<0 0 0 1>}
 lowerarm_twist_02_r {parent=lowerarm_r, loc=<-9.6199 0 0>, rot=<0 0 0 1>}
 upperarm_twist_01_r {parent=upperarm_r, loc=<-0.0041 0 0>, rot=<0 0 0 1>}
 deform_upperarm_bk_base_r {parent=upperarm_twist_01_r, loc=<-1.0815 7.2192 1.9456>, rot=<0 0 0 1>}
 deform_upperarm_bk_r {parent=deform_upperarm_bk_base_r, loc=<0 0 0>, rot=<0 0 0 1>}
 deform_upperarm_fr_base_r {parent=upperarm_twist_01_r, loc=<0.3689 -7.0506 2.7652>, rot=<0 0 0 1>}
 deform_upperarm_fr_r {parent=deform_upperarm_fr_base_r, loc=<0 0 0>, rot=<0 0 0 1>}
 deform_upperarm_up_base_r {parent=upperarm_twist_01_r, loc=<3.2213 1.1391 -3.3308>, rot=<0 0 0 1>}
 deform_upperarm_up_r {parent=deform_upperarm_up_base_r, loc=<0 0 0>, rot=<0 0 0 -1>}
 upperarm_twist_02_r {parent=upperarm_r, loc=<-13.7942 0 0>, rot=<0 0 0 1>}
 neck_01 {parent=spine_05, loc=<11.3138 0.1529 0>, rot=<0 0 -0.2003 -0.9797>}
 neck_02 {parent=neck_01, loc=<4.5627 0 0>, rot=<0 0 0.0504 -0.9987>}
 head {parent=neck_02, loc=<5.3914 0 0>, rot=<0 0 0.115 -0.9934>}
 dyn_helmStrap_01_Base_l {parent=head, loc=<4.8799 0.7281 -12.3043>, rot=<0.739 0.011 0.2501 -0.6254>}
 dyn_helmStrap_01_l {parent=dyn_helmStrap_01_Base_l, loc=<0 0 0>, rot=<0 0 0 1>}
 dyn_helmStrap_02_l {parent=dyn_helmStrap_01_l, loc=<-3.0329 0.0001 0.0001>, rot=<0.0092 0.0426 -0.1407 -0.9891>}
 dyn_helmStrap_03_l {parent=dyn_helmStrap_02_l, loc=<-3.2673 -0.0001 -0.0001>, rot=<-0.0241 0.0036 -0.0442 -0.9987>}
 dyn_helmStrap_04_l {parent=dyn_helmStrap_03_l, loc=<-3.5642 0 -0.0001>, rot=<-0.076 -0.0221 0.0106 -0.9968>}
 dyn_helmStrap_05_l {parent=dyn_helmStrap_04_l, loc=<-1.2484 0.0001 0.0002>, rot=<-0.0085 0.0335 0.0164 -0.9993>}
 dyn_helmStrap_01_Base_r {parent=head, loc=<4.8796 0.7281 12.3043>, rot=<0.011 -0.739 0.6254 0.2501>}
 dyn_helmStrap_01_r {parent=dyn_helmStrap_01_Base_r, loc=<0 0 0>, rot=<0 0 0 1>}
 dyn_helmStrap_02_r {parent=dyn_helmStrap_01_r, loc=<3.033 0 0>, rot=<0.0092 0.0426 -0.1407 -0.9891>}
 dyn_helmStrap_03_r {parent=dyn_helmStrap_02_r, loc=<3.2668 0 0>, rot=<-0.0241 0.0036 -0.0442 -0.9987>}
 dyn_helmStrap_04_r {parent=dyn_helmStrap_03_r, loc=<3.5637 0 0>, rot=<-0.076 -0.0221 0.0106 -0.9968>}
 dyn_helmStrap_05_r {parent=dyn_helmStrap_04_r, loc=<1.249 0 0>, rot=<-0.0085 0.0335 0.0164 -0.9993>}
 dyn_strap_parent {parent=head, loc=<2.9043 -6.7156 -7.0857>, rot=<0.0029 -0.7071 -0.0029 -0.7071>}
 dyn_strap_driver {parent=dyn_strap_parent, loc=<0 0 0>, rot=<0 0 0 1>}
 dyn_strapA_01 {parent=dyn_strap_driver, loc=<-0.3433 -0.2377 -0.2761>, rot=<-0.1618 -0.3442 0.4801 -0.7905>}
 dyn_strapA_02 {parent=dyn_strapA_01, loc=<4.1094 0 0>, rot=<0.0281 -0.0097 0.0055 -0.9995>}
 dyn_strapA_03 {parent=dyn_strapA_02, loc=<7.6929 0 0>, rot=<0.0432 0.0034 0.0102 -0.999>}
 dyn_strapA_04 {parent=dyn_strapA_03, loc=<6.2207 0 0>, rot=<0.0397 0.0091 -0.0022 -0.9992>}
 dyn_strapA_05 {parent=dyn_strapA_04, loc=<6.3479 0 0>, rot=<0.0387 0.0032 -0.003 -0.9992>}
 dyn_strapB_01 {parent=dyn_strap_driver, loc=<0.4028 0.3274 -0.1608>, rot=<-0.3541 -0.3175 0.4097 -0.7784>}
 dyn_strapB_02 {parent=dyn_strapB_01, loc=<6.7977 0 0>, rot=<0.0903 0.0131 -0.0084 -0.9958>}
 dyn_strapB_03 {parent=dyn_strapB_02, loc=<4.5377 0 0>, rot=<0.0826 -0.0005 0.004 -0.9966>}
 dyn_strapB_04 {parent=dyn_strapB_03, loc=<4.571 0 0>, rot=<0.063 -0.007 -0.0083 -0.998>}
 dyn_strapB_05 {parent=dyn_strapB_04, loc=<7.0973 0 0>, rot=<0.0075 0.0012 0.0081 -0.9999>}
 dyn_lensCap_01_Base {parent=head, loc=<18.9517 15.3285 -15.5357>, rot=<0.9501 -0.0423 0.0882 -0.2961>}
 dyn_lensCap_01 {parent=dyn_lensCap_01_Base, loc=<0 0 0>, rot=<0 0 0 -1>}
 dyn_lensCap_02 {parent=dyn_lensCap_01, loc=<-1.5412 0 0>, rot=<0.004 0.0875 0.0276 -0.9958>}
 dyn_lensCap_03 {parent=dyn_lensCap_02, loc=<-1.5317 0 0>, rot=<0 0.0091 -0.0118 -0.9999>}
 dyn_lensCap_04 {parent=dyn_lensCap_03, loc=<-0.691 0 0>, rot=<-0.1392 0 0 -0.9903>}
 attach_cape {parent=spine_05, loc=<9.017 -11.714 0>, rot=<0.0616 -0.7044 -0.0616 -0.7044>}
 attach_backpack {parent=spine_05, loc=<-0.123 -13.473 0>, rot=<-0.0093 -0.707 0.0093 -0.707>}
 deform_collarBack_base {parent=spine_05, loc=<16.6435 -6.0954 -0.5662>, rot=<0.0284 -0.7065 -0.0284 -0.7065>}
 deform_collarBack {parent=deform_collarBack_base, loc=<0 0 0>, rot=<0 0 0 1>}
 Backpack_BR {parent=spine_05, type=Null, loc=<-0.1228 -13.4727 0>, rot=<0 0.7071 0 0.7071>}
 cape {parent=spine_05, type=Null, loc=<8.1107 -11.2629 -0.0416>, rot=<-0.0567 0.7048 0 0.7071>}
 Charm_Attach_00 {parent=spine_05, type=Null, loc=<-2.8083 11.1445 -10.9841>, rot=<-0.5642 0.3853 -0.6531 0.3266>}
 neckContact_reduceClipping_base {parent=spine_05, loc=<5.0315 2.1628 0>, rot=<0.0284 -0.7065 -0.0284 -0.7065>, scale=<0.5 0.5 0.5>}
 neckContact_reduceClipping {parent=neckContact_reduceClipping_base, loc=<0 0 0>, rot=<0 0 0 1>, scale=<2 2 2>}
 deform_collarFront_base {parent=spine_05, loc=<3.4302 15.104 -0.5662>, rot=<0.0284 -0.7065 -0.0284 -0.7065>}
 deform_collarFront {parent=deform_collarFront_base, loc=<0 0 0>, rot=<0 0 0 1>}
 deform_collarSideR_base {parent=spine_05, loc=<12.521 5.0364 13.138>, rot=<0.0284 -0.7065 -0.0284 -0.7065>}
 deform_collarSideR {parent=deform_collarSideR_base, loc=<0 0 0>, rot=<0 0 0 1>}
 deform_collarSideL_base {parent=spine_05, loc=<12.5214 5.0364 -13.138>, rot=<0.0284 -0.7065 -0.0284 -0.7065>}
 deform_collarSideL {parent=deform_collarSideL_base, loc=<0 0 0>, rot=<0 0 0 1>}
 deform_pec_aim_base_l {parent=spine_04, loc=<7.8732 12.6388 -2.4123>, rot=<0.0102 -0.4602 0.3979 -0.7936>}
 deform_pec_aim_l {parent=deform_pec_aim_base_l, loc=<0 0 0>, rot=<0 0 0 1>}
 deform_pec_base_l {parent=deform_pec_aim_l, loc=<10.766 0 0>, rot=<0 0 0 1>}
 deform_pec_l {parent=deform_pec_base_l, loc=<0.055 0 0>, rot=<0 0 0 1>}
 deform_pec_aim_base_r {parent=spine_04, loc=<7.8727 12.6388 2.4123>, rot=<-0.4602 -0.0102 0.7936 0.3979>}
 deform_pec_aim_r {parent=deform_pec_aim_base_r, loc=<0 0 0>, rot=<0 0 0 1>}
 deform_pec_base_r {parent=deform_pec_aim_r, loc=<-10.7662 -0.0006 -0.0007>, rot=<0 0 0 1>}
 deform_pec_r {parent=deform_pec_base_r, loc=<0 0 0>, rot=<0 0 0 1>}
 thigh_l {parent=pelvis, loc=<-2.5693 -0.0109 -7.7501>, rot=<0.0012 0.0572 0.0155 -0.9982>}
 calf_l {parent=thigh_l, loc=<-39.0484 0 0>, rot=<0.0006 -0.0161 -0.0391 -0.9991>}
 calf_twist_01_l {parent=calf_l, loc=<-35.813 -0.0967 0.0397>, rot=<0.0237 0.0095 -0.0035 -0.9997>}
 calf_twist_02_l {parent=calf_l, loc=<-18.8064 -0.0967 0.0397>, rot=<0.0237 0.0095 -0.0035 -0.9997>}
 dyn_lacesBase_l {parent=calf_twist_02_l, loc=<-2.4569 2.1654 -1.5814>, rot=<0 0 0 1>}
 dyn_loopIn_l {parent=dyn_lacesBase_l, loc=<-0.6966 0.0909 0.7255>, rot=<0.2337 -0.2552 0.9382 0.0032>}
 dyn_loopOut_l {parent=dyn_lacesBase_l, loc=<-0.8747 0.1514 -0.2064>, rot=<-0.0045 0.1193 0.9924 0.0287>}
 foot_l {parent=calf_l, loc=<-42.561 0.1077 0.4292>, rot=<0.0418 -0.0455 0.0343 -0.9975>}
 ball_l {parent=foot_l, loc=<-6.1599 13.8264 0.6806>, rot=<0.0001 -0.0001 -0.7063 -0.7079>}
 deform_foot_in_base_l {parent=foot_l, loc=<3.9714 4.154 0.1123>, rot=<0.5234 0.4729 0.5208 0.4807>}
 deform_foot_in_l {parent=deform_foot_in_base_l, loc=<0 0 0>, rot=<0 0 0 1>}
 deform_foot_out_base_l {parent=foot_l, loc=<1.2973 -5.7338 -0.5554>, rot=<0.4729 -0.5234 0.4807 -0.5208>}
 deform_foot_out_l {parent=deform_foot_out_base_l, loc=<0 0 0>, rot=<0 0 0 1>}
 deform_knee_out_base_l {parent=calf_l, loc=<0 4 0>, rot=<0 0 0 1>}
 deform_knee_out_l {parent=deform_knee_out_base_l, loc=<0 0 0>, rot=<0 0 0 1>}
 deform_knee_in_base_l {parent=calf_l, loc=<1.298 -5.2957 -0.0496>, rot=<0 0 0 1>}
 deform_knee_in_l {parent=deform_knee_in_base_l, loc=<0 0 0>, rot=<0 0 0 1>}
 thigh_twist_01_l {parent=thigh_l, loc=<-21.6303 0 0>, rot=<0.0013 0 0 -1>}
 deform_groin_base_l {parent=thigh_l, loc=<-2.1908 8.1142 -1.7442>, rot=<0 0 0 1>}
 deform_groin_l {parent=deform_groin_base_l, loc=<0 0 0>, rot=<0 0 0 1>}
 thigh_r {parent=pelvis, loc=<-2.5691 -0.0109 7.7501>, rot=<0.0572 -0.0012 0.9982 0.0155>}
 calf_r {parent=thigh_r, loc=<39.0487 0 0>, rot=<0.0006 -0.0161 -0.0391 -0.9991>}
 calf_twist_01_r {parent=calf_r, loc=<35.8132 0.0967 -0.0397>, rot=<0.0237 0.0095 -0.0035 -0.9997>}
 calf_twist_02_r {parent=calf_r, loc=<18.8064 0.0967 -0.0397>, rot=<0.0237 0.0095 -0.0035 -0.9997>}
 dyn_lacesBase_r {parent=calf_twist_02_r, loc=<2.4883 -2.1681 1.5781>, rot=<0 0 0 1>}
 dyn_loopIn_r {parent=dyn_lacesBase_r, loc=<0.6906 -0.0944 -0.6427>, rot=<-0.2508 -0.2378 0.0029 -0.9384>}
 dyn_loopOut_r {parent=dyn_lacesBase_r, loc=<0.8482 -0.1491 0.2092>, rot=<0.1193 0.0045 0.0287 -0.9924>}
 foot_r {parent=calf_r, loc=<42.5612 -0.1077 -0.4292>, rot=<0.0418 -0.0455 0.0343 -0.9975>}
 ball_r {parent=foot_r, loc=<6.1598 -13.8264 -0.6806>, rot=<0.0001 -0.0001 -0.7063 -0.7079>}
 deform_foot_in_base_r {parent=foot_r, loc=<-3.9715 -4.154 -0.1118>, rot=<-0.5208 0.4807 0.5234 -0.4729>}
 deform_foot_in_r {parent=deform_foot_in_base_r, loc=<0 0 0>, rot=<0 0 0 1>}
 deform_foot_out_base_r {parent=foot_r, loc=<-1.2975 5.7337 0.5559>, rot=<0.4807 0.5208 -0.4729 -0.5234>}
 deform_foot_out_r {parent=deform_foot_out_base_r, loc=<0 0 0>, rot=<0 0 0 1>}
 deform_knee_in_base_r {parent=calf_r, loc=<-1.2981 5.2958 0.0496>, rot=<0 0 0 1>}
 deform_knee_in_r {parent=deform_knee_in_base_r, loc=<0 0 0>, rot=<0 0 0 1>}
 deform_knee_out_base_r {parent=calf_r, loc=<0 -4 0>, rot=<0 0 0 1>}
 deform_knee_out_r {parent=deform_knee_out_base_r, loc=<0 0 0>, rot=<0 0 0 1>}
 thigh_twist_01_r {parent=thigh_r, loc=<21.6303 0 0>, rot=<0.0013 0 0 -1>}
 deform_groin_base_r {parent=thigh_r, loc=<2.1909 -8.1142 1.7442>, rot=<0 0 0 1>}
 deform_groin_r {parent=deform_groin_base_r, loc=<0 0 0>, rot=<0 0 0 1>}
 dyn_pouchB_Base_l {parent=pelvis, loc=<9.4583 -3.5409 -14.4354>, rot=<0.0997 -0.8437 -0.5171 -0.1044>}
 dyn_pouchB_l {parent=dyn_pouchB_Base_l, loc=<0 0 0>, rot=<0 0 0 1>}
 dyn_pouchA_Base_l {parent=pelvis, loc=<9.1319 4.2527 -12.4662>, rot=<-0.0733 0.437 0.8909 0.0999>}
 dyn_pouchA_l {parent=dyn_pouchA_Base_l, loc=<0 0 0>, rot=<0 0 0 1>}
 dyn_pouchA_Base_r {parent=pelvis, loc=<9.3905 5.6754 11.2284>, rot=<0.091 -0.4194 0.9004 0.0711>}
 dyn_pouchA_r {parent=dyn_pouchA_Base_r, loc=<0 0 0>, rot=<0 0 0 1>}
 dyn_pouchBack_Base {parent=pelvis, loc=<11.9842 -11.1688 0>, rot=<0.0438 -0.9984 0.0349 -0.0015>}
 dyn_pouchBack {parent=dyn_pouchBack_Base, loc=<0 0 0>, rot=<0 0 0 1>}
 dyn_pouchB_Base_r {parent=pelvis, loc=<9.9583 -1.3955 13.1688>, rot=<0.12 -0.7501 0.6408 0.1109>}
 dyn_pouchB_r {parent=dyn_pouchB_Base_r, loc=<0 0 0>, rot=<0 0 0 1>}
 deform_glute_base_l {parent=pelvis, loc=<-8.8064 -8.8222 -8.4287>, rot=<0.0275 0.0501 -0.4472 -0.8926>}
 deform_glute_l {parent=deform_glute_base_l, loc=<0 0 0>, rot=<0 0 0 1>}
 deform_glute_base_r {parent=pelvis, loc=<-8.8062 -8.8222 8.4287>, rot=<0.0501 -0.0275 0.8926 -0.4472>}
 deform_glute_r {parent=deform_glute_base_r, loc=<0 0 0>, rot=<0 0 0 1>}
 ik_foot_root {parent=root, loc=<0 0 0>, rot=<0 0 0 1>}
 ik_foot_l {parent=ik_foot_root, loc=<15.2608 -4.4176 9.4433>, rot=<0.0357 0.7045 0.0284 -0.7082>}
 ik_foot_r {parent=ik_foot_root, loc=<-15.2609 -4.4176 9.4431>, rot=<0.7082 0.0284 -0.7045 0.0357>}
 ik_hand_root {parent=root, loc=<0 0 0>, rot=<0 0 0 1>}
 ik_hand_gun {parent=ik_hand_root, loc=<-51.973 4.089 102.694>, rot=<0.6084 0.1793 0.4613 -0.6205>}
 ik_hand_l {parent=ik_hand_gun, loc=<53.0284 82.1785 35.2051>, rot=<-0.5102 -0.7906 -0.3387 0>}
 ik_hand_r {parent=ik_hand_gun, loc=<0 0 0>, rot=<0 0 0 1>}
]
