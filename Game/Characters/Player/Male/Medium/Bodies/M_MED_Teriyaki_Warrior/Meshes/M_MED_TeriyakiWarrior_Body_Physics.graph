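# Physics asset for the Teriyaki Warrior medium male player body, set up for dangling pouch secondary motion rather than a full-body ragdoll.
# Five pouch boxes (A and B on each side, plus back) are each constrained to a tiny base box with limited swing and locked twist.
# Capsules on pelvis, thighs and lower calf twist bones provide body collision shapes, with an extra small capsule on the left thigh.
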PhysicsAsset /Game/Characters/Player/Male/Medium/Bodies/M_MED_Teriyaki_Warrior/Meshes/M_MED_TeriyakiWarrior_Body_Physics
bodies {
 calf_twist_02_l [{shapeType=Capsule, length=17.1007, radius=4.7425, center=<-10.6715 -2.4753 -0.7079>, rotation=<rot 83.0669 0 3.4565>}]
 calf_twist_02_r [{shapeType=Capsule, length=16.256, radius=5.5872, center=<7.9363 4.0317 1.0945>, rotation=<rot 77.7609 0 0>}]
 dyn_pouchA_Base_l [{shapeType=Box, extentX=1.01, extentY=1.01, extentZ=1.01, rotation=<rot 0 -90 0>}]
 dyn_pouchA_Base_r [{shapeType=Box, extentX=1.01, extentY=1.01, extentZ=1.01, rotation=<rot 0 -90 0>}]
 dyn_pouchA_l [{shapeType=Box, extentX=5.361, extentY=13.8535, extentZ=8.4654, center=<6.8954 -2.1358 0>, rotation=<rot 0 90 0>}]
 dyn_pouchA_r [{shapeType=Box, extentX=5.2282, extentY=10.1671, extentZ=8.0053, center=<4.56 -1.8172 -0.0001>, rotation=<rot 0 90 0>}]
 dyn_pouchB_Base_l [{shapeType=Box, extentX=1.01, extentY=1.01, extentZ=1.01, rotation=<rot 0 -90 0>}]
 dyn_pouchB_Base_r [{shapeType=Box, extentX=1.01, extentY=1.01, extentZ=1.01, rotation=<rot 0 -90 0>}]
 dyn_pouchB_l [{shapeType=Box, extentX=5.4605, extentY=13.6602, extentZ=8.3629, center=<6.9874 -0.459 0>, rotation=<rot 0 90 0>}]
 dyn_pouchB_r [{shapeType=Box, extentX=5.2199, extentY=10.8026, extentZ=8.3894, center=<4.56 -1.1147 0.0001>, rotation=<rot 0 90 0>}]
 dyn_pouchBack [{shapeType=Box, extentX=8.6941, extentY=15.3427, extentZ=18.6283, center=<5.063 -2.3681 0>, rotation=<rot 0 85.0807 0>}]
 dyn_pouchBack_Base [{shapeType=Box, extentX=1.01, extentY=1.01, extentZ=1.01, rotation=<rot 0 -90 0>}]
 pelvis [{shapeType=Capsule, length=7.2269, radius=10.3537, rotation=<rot 0 -90 0>}]
 thigh_l [
  {shapeType=Capsule, length=33.1969, radius=8.7909, center=<-19.2557 0.0003 0>, rotation=<rot 90 0 0>}
  {shapeType=Capsule, length=11.7968, radius=3.843, shapeName=thigh_l_capsule, center=<-6.2781 0 -5.0082>, rotation=<rot -18.8219 0 0>}
 ]
 thigh_r [{shapeType=Capsule, length=32.8901, radius=7.7134, center=<18.0833 0 0>, rotation=<rot 90 0 0>}]
}
constraints [
 {bone1Name=dyn_pouchB_l, bone2Name=dyn_pouchB_Base_l, swing1LimitDegrees=20, swing2LimitDegrees=12, twistLimitDegrees=45, twistMotion=Locked}
 {bone1Name=dyn_pouchA_l, bone2Name=dyn_pouchA_Base_l, swing1LimitDegrees=30, swing2LimitDegrees=12, twistLimitDegrees=45, twistMotion=Locked}
 {bone1Name=dyn_pouchA_r, bone2Name=dyn_pouchA_Base_r, swing1LimitDegrees=30, swing2LimitDegrees=12, twistLimitDegrees=45, twistMotion=Locked}
 {
  bone1Name=dyn_pouchBack
  bone2Name=dyn_pouchBack_Base
  swing1LimitDegrees=30
  swing2LimitDegrees=12
  twistLimitDegrees=45
  twistMotion=Locked
 }
 {bone1Name=dyn_pouchB_r, bone2Name=dyn_pouchB_Base_r, swing1LimitDegrees=30, swing2LimitDegrees=12, twistLimitDegrees=45, twistMotion=Locked}
]
